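# Tar Pit Trap bubble material for mesh particles, using Default Lit shading and world-space normals.
# Cubemap highlights tinted by time of day and a flattened foam normal blend in through a vertical-gradient mask.
# Panned, distorted world-space noise mixes via Fresnel, particle color and dynamic parameters, with day/night variants.
# The graph also builds a rainbow rim and a noise-driven vector along vertex normals, though output connections are not shown.
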Material /Game/Weapons/FORT_TRAPS/Tar_Pit_Trap/Materials/M_tar_bubble
graph {
 properties {
  bTangentSpaceNormal=false
  bUsedWithMeshParticles=true
  bUsedWithNiagaraMeshParticles=true
  opacityMaskClipValue=0.933
  pixelDepthOffsetMode=PDOM_Legacy
  previewMesh=/Game/Effects/Fort_Effects/Mesh/Shockwave/SM_Shockwave_Partial_Sphere
  translucencyLightingMode=TLM_Surface
  translucencyPass=MTP_BeforeDOF
  translucentBackscatteringExponent=550
  translucentSelfShadowDensityScale=5
  translucentSelfShadowSecondOpacity=10
  translucentShadowDensityScale=1
  usedShadingModels=Default Lit
 }
 nodes [
  n1 = Multiply(
   A=TextureSampleParameterCube!(UVs=ReflectionVectorWS(), Apply View MipBias) {parameterName=Cube, samplerType=SAMPLERTYPE_LinearColor, texture=/Game/Weapons/WeaponSkins/Textures/AthenaCubemap2}.RGB
   B=Multiply(
    A=MaterialFunctionCall() {materialFunction=&Color_Time_of_Day}.Result
    B=VectorParameter() {parameterName=HilightColor, defaultValue=<rgba 0.5 0.5 0.6 0>}.RGB
   )
  )
  n2 = ScalarParameter() {defaultValue=1, parameterName=HiLightBrightness}
  n3 = OneMinus(None=MaterialFunctionCall() {materialFunction=&LinearGradient}.VGradient)
  n4 = ScalarParameter() {defaultValue=2, parameterName=NormalTransMult}
  n5 = Saturate(None=Power(Base=Multiply(A=@n3, B=@n4), Exp=ScalarParameter() {defaultValue=2, parameterName=NormalTransExp}))
  Add(A=LinearInterpolate(B=Multiply(A=@n1, B=@n2), Alpha=@n5), B=VectorParameter() {parameterName=Color, defaultValue=<rgba 0 0 0.001 1>}.RGB)
  Add()
  Constant() {r=5000}
  n6 = MaterialFunctionCall(
   Normal=TextureSample() {samplerType=SAMPLERTYPE_Normal, texture=/Game/Packages/Fortress_Transylvania/Haunted_Forest/Swamps/Textures/T_sea_foam_03_n}.RGB
   Flatness=Constant() {r=0.8}
  ) {materialFunction=&FlattenNormal}
  n7 = Transform(None=@n6.Result)
  n8 = Constant3Vector() {constant=<rgba 1 1 0 0>}
  LinearInterpolate(A=Constant3Vector() {constant=<rgba 0 0 1 0>}, B=Add(A=Multiply(A=@n7, B=@n8), B=VertexNormalWS()), Alpha=@n5)
  n9 = MaterialFunctionCall() {materialFunction=&Fresnel_Function}
  n10 = Clamp(None=Multiply(A=Power(Base=@n9.Result)))
  n11 = ComponentMask(None=Clamp(None=CameraVectorWS())) {b=true}
  LinearInterpolate(A=LinearInterpolate(Alpha=@n10) {constA=0.2, constB=0.1}, Alpha=@n11)
  n12 = Divide(A=ComponentMask(None=WorldPosition() {worldPositionShaderOffset=WPT_ExcludeAllShaderOffsets}.XYZ) {g=true, r=true}) {constB=200}
  n13 = ComponentMask(None=TextureSample(UVs=Panner(Coordinate=@n12) {speedY=-0.08}) {texture=&2ndInfectedMapLinear}.R) {g=true, r=true}
  n14 = WorldPosition() {worldPositionShaderOffset=WPT_ExcludeAllShaderOffsets}
  n15 = Add(A=Multiply(A=@n13) {constB=0.025}, B=Divide(A=ComponentMask(None=@n14.XYZ) {g=true, r=true}) {constB=500})
  n16 = Multiply(A=CameraVectorWS()) {constB=0.3}
  n17 = TextureSample(UVs=Add(A=Panner(Coordinate=@n15) {speedY=0.03}, B=ComponentMask(None=@n16) {g=true, r=true})) {texture=/Game/Packages/Fortress_Transylvania/Haunted_Forest/Swamps/Textures/2ndInfectedMapLinear2}
  n18 = Multiply(A=@n13) {constB=0.05}
  n19 = Panner(Coordinate=TextureCoordinate() {uTiling=4, vTiling=4}) {speedY=-0.25}
  n20 = DynamicParameter() {defaultValue=<rgba 0 0 0 1>, paramNames [Param1,Param2,VTiling,UTiling]}
  n21 = Power(Base=TextureSample(UVs=Add(A=@n18, B=@n19)) {texture=/Game/Global/Textures/Default/Tileable_Patterns/T_Tornado_Tile}.RGB, Exp=@n20.Param2)
  n22 = Multiply(A=LinearInterpolate(A=@n17.R, B=Multiply(A=@n17.R, B=@n21), Alpha=@n20.Param1), B=LinearInterpolate(Alpha=@n11) {constA=1, constB=0.01}) {constB=0.5}
  n23 = Multiply(A=LinearInterpolate(Alpha=@n11) {constA=1, constB=0.3}) {constB=0.5}
  n24 = ParticleColor()
  n25 = Clamp(None=Multiply(A=LinearInterpolate(A=@n22, B=@n23, Alpha=@n10), B=@n24.RGB))
  n26 = CollectionParameter() {
   collection=&FortniteMaterialParameters
   parameterName=FortniteActiveTimeOfDay
  }
  n27 = DotProduct(A=@n26, B=Constant4Vector() {constant=<rgba 1 1 1 0>}.RGBA)
  LinearInterpolate(A=@n25, B=Desaturation(None=@n25, Fraction=Constant() {r=0.25}), Alpha=@n27) {desc="A=Night B=Day"}
  LinearInterpolate(A=ScalarParameter() {defaultValue=1, parameterName=Rough}, B=ScalarParameter() {defaultValue=1, parameterName=RoughHigh}, Alpha=@n5)
  n28 = Multiply(A=@n24.RGB) {constB=0.5}
  LinearInterpolate(A=Multiply(A=@n28, B=Constant3Vector() {constant=<rgba 2 1.2709 0.6762 0>}) {constB=30}, B=@n28, Alpha=@n27) {constB=0, desc="A=Night B=Day"}
  LinearInterpolate(Alpha=@n27) {desc="A=Night B=Day"}
  n29 = LinearInterpolate(Alpha=@n10) {constB=0.9}
  n30 = Constant() {r=0.9}
  n31 = Constant() {r=0.9}
  Multiply(
   A=LinearInterpolate(A=@n29, B=Multiply(A=@n29) {constB=0.1}, Alpha=@n11) {constB=0}
   B=LinearInterpolate(A=Desaturation(None=@n24.RGB, Fraction=@n30), B=Desaturation(None=@n24.RGB, Fraction=@n31), Alpha=@n27) {desc="A=Night B=Day"}
  )
  n32 = DotProduct(A=CameraVectorWS(), B=VertexNormalWS())
  Multiply(
   A=Power(Base=OneMinus(None=@n32), Exp=ScalarParameter() {defaultValue=3, parameterName=Exp})
   B=Multiply(
    A=TextureSample() {samplerType=SAMPLERTYPE_LinearColor, texture=/Game/Athena/Environments/Festivus/Materials/T_Rainbow_02_D}.RGB
    B=ScalarParameter() {defaultValue=0.05, parameterName=ColorMult}
   )
  )
  n33 = MaterialFunctionCall() {materialFunction=&LinearGradient}
  n34 = Constant()
  n35 = Add(A=Multiply(A=Cosine(None=@n33.VGradient), B=@n34), B=@n34)
  n36 = WorldPosition() {worldPositionShaderOffset=WPT_ExcludeAllShaderOffsets}
  n37 = Panner(Coordinate=Divide(A=ComponentMask(None=@n36.XYZ) {g=true, r=true}) {constB=800}) {speedY=0.1}
  n38 = Constant() {r=15}
  n39 = Multiply(A=Multiply(A=Divide(A=@n36.XYZ) {constB=800}) {constB=0.02}, B=@n38)
  Multiply(
   A=Multiply(A=OneMinus(None=@n35), B=TextureSample(UVs=@n37) {texture=&2ndInfectedMapLinear}.RGB)
   B=Multiply(A=VertexNormalWS(), B=Add(A=@n39, B=@n38))
  )
  Normalize(VectorInput=@n6.Result)
  ParticleColor()
  ScalarParameter() {defaultValue=1, parameterName=Metal}
  ScalarParameter() {defaultValue=0.3, parameterName=Spec}
 ]
}
